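# Niagara system for the Scholar backpack cosmetic: looping fire in two User-driven colors, purple by default, with sparks, airglow and light.
# Most emitters split particles across two skeletal mesh sockets via spawn groups; Fire_Light adds one offset light.
# Fire meshes, glow and light are persistent looping particles with flicker curves, while flame, decal and spark sprites spawn continuously.
# A system-level FN_Integrator normalized value scales brightness, size, lift and spawn probability, varying the fire's intensity.
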NiagaraSystem /BRCosmetics/Accessories/FORT_Backpacks/Backpack_F_MED_Scholar/FX/NS_ScholarBackpack
dependencies {
 usedDataInterfaces [&NiagaraDataInterfaceCurve, &NiagaraDataInterfaceColorCurve]
 usedDynamicInputs []
 usedModules [
  &self:SystemSpawnScript.SystemScriptSource.SystemScriptGraph.NiagaraNodeAssignment_0.SetVariables_#1
  &FN_Integrator
  &SystemState
 ]
}
emitters {
 FireAirglow {
  emitterUpdateScript [
   &EmitterState(Life Cycle Mode=::NewEnumerator1, UseLoopDelay=false)
   &SpawnBurst_Instantaneous(Spawn Count=2, Spawn Probability=HlslExpression() {hlslExpression="Emitter.LoopCount < 1"}, Use Spawn Probability=true)
  ]
  particleSpawnScript [
   &FN_SetSpawnGroup(Number of Spawn Groups=2)
   /Niagara/Modules/Spawn/Initialization/InitializeParticle(
    Color=<rgba 0.6445 0.1981 0.6514 1>
    Lifetime=0.8
    Module.Write Color=true
    Module.Write Lifetime=true
    Module.Write Mass=true
    Module.Write Material Random=true
    Module.Write Position=false
    Module.Write Scale=false
    Module.Write SpriteRotation=false
    Module.Write SpriteSize=true
    Sprite Size=&Multiply_Vector2(A=<50 40>, B=&Vector2DFromFloat(Value=0.8))
   )
  ]
  particleUpdateScript [
   &ParticleState(
    Kill Particles When Lifetime Has Elapsed=false
    Let Infinitely Lived Particles Die When Emitter Deactivates=true
    Loop Particles Lifetime=true
   )
   &SkeletalMeshLocation(
    Apply Sampled Orientation to Mesh Particles=true
    Bone / Socket Index=$Particles.SpawnGroup
    Inherit Velocity=false
    MaskSpawnGroup=false
    Mesh Sampling Type=::NewEnumerator5
    Sampled Orientation Coordinate Space=::Local
    Sampled Position Coordinate Space=::Local
    Sampled Position Offset Coordinate Space:ENiagaraMeshTransforms=::NewEnumerator2
    Sampled Velocity Coordinate Space:ENiagaraCoordinateSpace=::Local
    Skeletal Mesh=$User.SkeletalMesh
    Socket Sampling Mode=::NewEnumerator5
   ) {module=SkeletalMeshLocation002}
   &Color(Color=$User.Color_01)
   &ScaleColor(
    Color Value To Scale=$Particles.Color
    Scale Alpha=&FloatFromCurve(
     FloatCurve=DataInterface() {propertyValues {Curve {keys [
         <at 0 0 RCIM_Cubic>
         {arriveTangent=0.8515, interpMode=RCIM_Cubic, leaveTangent=0.8515, time=0.1275, value=1}
         {arriveTangent=-1.1461, interpMode=RCIM_Cubic, leaveTangent=-1.1461, time=0.8566, value=0.7294}
         <at 1 0>
        ]}}}
     Scale Curve=2
    )
    Scale RGB=&VectorFromFloat(Value=&FloatFromCurve(
      FloatCurve=DataInterface() {propertyValues {Curve {keys [<at 0 1 RCIM_Cubic>, {arriveTangent=0.5, interpMode=RCIM_Cubic, leaveTangent=0.5, time=0.8592, value=1.35}, <at 1 1.5 RCIM_Cubic>]}}}
      CurveIndex=$System.FN_Integrator.NormalizedValue
     ))
    ScaleA=false
    ScaleRGB=true
   ) {module=ScaleColor001}
   /SetVariables(
    Particles.MeshColor:LinearColor=&MultiplyLinearColor(Linear Color=$Particles.Color, Scale Factor=<1 1 1 1>)
    Particles.SpriteColor=&MultiplyLinearColor(
     Linear Color=$Particles.Color
     Scale Factor=&MakeFloat4FromVectorAndFloat(
      "Vector (RGB)"=<1 1 1>
      "Float (Alpha)"=&FloatFromCurve(
       FloatCurve=DataInterface() {propertyValues {Curve {keys [
           <at 0 1 RCIM_Cubic>
           {arriveTangent=-1.1468, interpMode=RCIM_Cubic, leaveTangent=-1.1468, time=0.0926, value=1.134}
           {arriveTangent=0.5291, interpMode=RCIM_Cubic, leaveTangent=0.5291, time=0.2116, value=0.7573}
           {arriveTangent=0.0266, interpMode=RCIM_Cubic, leaveTangent=0.0266, time=0.3435, value=1.2667}
           {arriveTangent=-0.1747, interpMode=RCIM_Cubic, leaveTangent=-0.1747, time=0.5441, value=0.7662}
           {arriveTangent=2.9922, interpMode=RCIM_Cubic, leaveTangent=2.9922, time=0.6139, value=1.2195}
           {arriveTangent=-1.5511, interpMode=RCIM_Cubic, leaveTangent=-1.5511, time=0.722, value=1.2984}
           {arriveTangent=-1.0733, interpMode=RCIM_Cubic, leaveTangent=-1.0733, time=0.8433, value=0.8636}
           <at 1 1 RCIM_Cubic>
          ]}}}
       Scale Curve=0.007
      )
     )
    )
   ) {isSetParameters=true, module=SetVariables_#2}
   &DynamicMaterialParameters(
    Index 0 Param 2=1
    Index 0 Param 3=12
    Module.Param0WriteEnabled=true
    Module.Param1WriteEnabled=true
    Module.Param2WriteEnabled=true
    Module.Param3WriteEnabled=true
   )
   &FN_Light_Attributes(
    Color=&LerpLinearColors(StartColor=$System.Color_01, EndColor=$System.Color_02, LerpFactor=0.8)
    Create Color Attribute=true
    Create Exponent Attribute=true
    Create Radius Attribute=true
    Enable Particle Light Enable=true
    Exponent=3
    Light Enabled=true
    Radius=&FloatFromCurve(
     FloatCurve=DataInterface() {propertyValues {Curve {keys [<at 0 1 RCIM_Cubic>, {arriveTangent=0.4, interpMode=RCIM_Cubic, leaveTangent=0.4, time=0.6732, value=1.1856}, <at 1 1.4 RCIM_Cubic>]}}}
     CurveIndex=$System.FN_Integrator.NormalizedValue
     Scale Curve=40
    )
    Scale Color=&Multiply_Float(
     A=&FloatFromCurve(
      FloatCurve=DataInterface() {propertyValues {Curve {keys [
          <at 0 1 RCIM_Cubic>
          {arriveTangent=-0.7971, interpMode=RCIM_Cubic, leaveTangent=-0.7971, time=0.0926, value=1.134}
          {arriveTangent=0.0731, interpMode=RCIM_Cubic, leaveTangent=0.0731, time=0.2029, value=0.8383}
          {arriveTangent=0.823, interpMode=RCIM_Cubic, leaveTangent=0.823, time=0.286, value=1.1481}
          {arriveTangent=0.2524, interpMode=RCIM_Cubic, leaveTangent=0.2524, time=0.3706, value=0.9763}
          {arriveTangent=-0.2615, interpMode=RCIM_Cubic, leaveTangent=-0.2615, time=0.7155, value=1.2565}
          {arriveTangent=-0.9017, interpMode=RCIM_Cubic, leaveTangent=-0.9017, time=0.8612, value=0.8481}
          <at 1 1 RCIM_Cubic>
         ]}}}
      Scale Curve=4
     )
     B=&FloatFromCurve(
      FloatCurve=DataInterface() {propertyValues {Curve {keys [<at 0 2 RCIM_Cubic>, {arriveTangent=4, interpMode=RCIM_Cubic, leaveTangent=4, time=0.6669, value=3.2773}, <at 1 6 RCIM_Cubic>]}}}
      CurveIndex=$System.FN_Integrator.NormalizedValue
     )
    )
   )
   &CameraOffset(Camera Offset Amount=-5)
  ]
  rendererClasses [&NiagaraSpriteRendererProperties]
  renderers [{rendererClass=&NiagaraSpriteRendererProperties}]
 }
 FireMesh {
  emitterUpdateScript [
   &EmitterState(Life Cycle Mode=::NewEnumerator1, Loop Duration=10, UseLoopDelay=false)
   &SpawnBurst_Instantaneous(Spawn Count=2, Spawn Probability=HlslExpression() {hlslExpression="Emitter.LoopCount == 0"}, Use Spawn Probability=true)
  ]
  particleSpawnScript [
   &InitializeParticle(
    Color=<rgba 0.6445 0.1981 0.6514 1>
    Lifetime=10
    Module.Write Color=true
    Module.Write Lifetime=true
    Module.Write Mass=true
    Module.Write Material Random=true
    Module.Write Position=false
    Module.Write Scale=false
    Module.Write SpriteRotation=false
    Module.Write SpriteSize=true
    Sprite Size=&Multiply_Vector2(A=<50 30>, B=&Vector2DFromFloat(Value=0.8))
   )
   &FN_SetSpawnGroup(Number of Spawn Groups=2)
   &InitialMeshOrientation(
    Rotation=&RotateVectorByQuaternion(
     Vector To Rotate=&Lerp_Vector(A=<0 0 0.25>, B=<0 0 -0.25>, Alpha=&MakeFloatFromInt(Int=$Particles.SpawnGroup))
     Quaternion=$Output.SkeletalMeshLocation.SampledRotation
    )
    Use Orientation Vector=true
    Use Rotation Vector=true
   )
  ]
  particleUpdateScript [
   &ParticleState(
    Kill Particles When Lifetime Has Elapsed=false
    Let Infinitely Lived Particles Die When Emitter Deactivates=true
    Loop Particles Lifetime=true
   )
   &SkeletalMeshLocation(
    Apply Sampled Orientation to Mesh Particles=true
    Bone / Socket Index=$Particles.SpawnGroup
    Inherit Velocity=false
    MaskSpawnGroup=false
    Mesh Sampling Type=::NewEnumerator5
    Sampled Orientation Coordinate Space=::Local
    Sampled Position Coordinate Space=::Local
    Sampled Position Offset=&Lerp_Vector(A=<0 0 0>, B=<0 0 0>, Alpha=&MakeFloatFromInt(Int=$Particles.SpawnGroup))
    Sampled Position Offset Coordinate Space:ENiagaraMeshTransforms=::NewEnumerator2
    Sampled Velocity Coordinate Space:ENiagaraCoordinateSpace=::Local
    Skeletal Mesh=$User.SkeletalMesh
    Socket Sampling Mode=::NewEnumerator5
   ) {module=SkeletalMeshLocation002}
   &Color(Color=$User.Color_01)
   &ScaleColor(
    Color Value To Scale=$Particles.Color
    Scale Alpha=&FloatFromCurve(
     FloatCurve=DataInterface() {propertyValues {Curve {keys [
         <at 0 0 RCIM_Cubic>
         {arriveTangent=0.8515, interpMode=RCIM_Cubic, leaveTangent=0.8515, time=0.1275, value=1}
         {arriveTangent=-1.1461, interpMode=RCIM_Cubic, leaveTangent=-1.1461, time=0.8566, value=0.7294}
         <at 1 0>
        ]}}}
     Scale Curve=2
    )
    Scale RGB=&VectorFromFloat(Value=&FloatFromCurve(
      FloatCurve=DataInterface() {propertyValues {Curve {keys [<at 0 1 RCIM_Cubic>, {arriveTangent=0.8, interpMode=RCIM_Cubic, leaveTangent=0.8, time=0.6771, value=1.1705}, <at 1 1.8 RCIM_Cubic>]}}}
      CurveIndex=$System.FN_Integrator.NormalizedValue
      Scale Curve=1.25
     ))
    ScaleA=false
    ScaleRGB=true
   ) {module=ScaleColor001}
   &ScaleMeshSize(Scale Factor=&Lerp_Vector(A=<0.9 1 0.3>, B=<0.9 1 0.65>, Alpha=$System.FN_Integrator.NormalizedValue))
   &DynamicMaterialParameters(
    Index 0 Param 1=&RandomRangeFloat(Minimum=0, Maximum=1)
    Index 0 Param 2=1
    Index 0 Param 3=14
    Index 1 Param 4=&Multiply_Float(
     A=&FloatFromCurve(
      FloatCurve=DataInterface() {propertyValues {Curve {keys [
          <at 0 1 RCIM_Cubic>
          {arriveTangent=-1.3184, interpMode=RCIM_Cubic, leaveTangent=-1.3184, time=0.0958, value=1.2356}
          {arriveTangent=0.0731, interpMode=RCIM_Cubic, leaveTangent=0.0731, time=0.2145, value=0.7171}
          {arriveTangent=1.6609, interpMode=RCIM_Cubic, leaveTangent=1.6609, time=0.2892, value=1.2497}
          {arriveTangent=0.2524, interpMode=RCIM_Cubic, leaveTangent=0.2524, time=0.3706, value=0.9763}
          {arriveTangent=-0.265, interpMode=RCIM_Cubic, leaveTangent=-0.265, time=0.7188, value=1.3581}
          {arriveTangent=-1.2732, interpMode=RCIM_Cubic, leaveTangent=-1.2732, time=0.8508, value=0.8491}
          <at 1 1 RCIM_Cubic>
         ]}}}
      Scale Curve=2
     )
     B=&FloatFromCurve(
      FloatCurve=DataInterface() {propertyValues {Curve {keys [<at 0 0.5 RCIM_Cubic>, {arriveTangent=1.3, interpMode=RCIM_Cubic, leaveTangent=1.3, time=0.6353, value=1.25}, <at 1 1.8 RCIM_Cubic>]}}}
      CurveIndex=$System.FN_Integrator.NormalizedValue
     )
    )
    Module.Param0WriteEnabled=true
    Module.Param0WriteEnabled1=true
    Module.Param1WriteEnabled=true
    Module.Param1WriteEnabled1=true
    Module.Param2WriteEnabled=true
    Module.Param2WriteEnabled1=true
    Module.Param3WriteEnabled=true
    Module.Param3WriteEnabled1=true
    Write Parameter Index 1=true
   )
   /SetVariables(Particles.VisibilityTag=&MakeIntFromFloat(Float=$User.RenderVis)) {isSetParameters=true}
  ]
  rendererClasses [&NiagaraMeshRendererProperties]
  renderers [
   {rendererClass=&NiagaraMeshRendererProperties, rendererIndex=0}
   {rendererClass=&NiagaraMeshRendererProperties, rendererIndex=1}
  ]
 }
 Fire_Decals {
  emitterUpdateScript [
   &EmitterState()
   &SpawnRate(
    Spawn Probability=&FloatFromCurve(FloatCurve=DataInterface() {propertyValues {Curve {keys [<at 0 1>,<at 1 0.75>]}}}, CurveIndex=$System.FN_Integrator.NormalizedValue)
    SpawnRate=20
    Use Spawn Probability=true
   )
  ]
  particleSpawnScript [
   &InitializeParticle(
    Color=&RandomRangeLinearColor(Minimum=<rgba 0 0 0 1>, Maximum=<rgba 1 1 1 1>)
    Lifetime=&UniformRangedFloat(Minimum=0.4, Maximum=0.6)
    Module.Write Color=false
    Module.Write Lifetime=true
    Module.Write Mass=false
    Module.Write Material Random=true
    Module.Write Position=false
    Module.Write Scale=false
    Module.Write SpriteRotation=true
    Module.Write SpriteSize=true
    Position=&SimulationPosition()
    Sprite Rotation=&UniformRangedFloat(Minimum=0, Maximum=0)
    Sprite Size=&Vector2DFromFloat(Value=&RandomRangeFloat(Minimum=20, Maximum=25))
   )
   &FN_SetSpawnGroup(Number of Spawn Groups=2)
   &SkeletalMeshLocation(
    Apply Sampled Orientation to Mesh Particles=true
    Bone / Socket Index=$Particles.SpawnGroup
    Inherit Velocity=false
    MaskSpawnGroup=false
    Mesh Sampling Type=::NewEnumerator5
    Sampled Orientation Coordinate Space=::Local
    Sampled Position Coordinate Space=::Local
    Sampled Position Offset=&RandomRangeVector(Minimum=<0 -8 0>, Maximum=<0 8 0>)
    Sampled Position Offset Coordinate Space:ENiagaraMeshTransforms=::NewEnumerator3
    Sampled Velocity Coordinate Space:ENiagaraCoordinateSpace=::Local
    Skeletal Mesh=$User.SkeletalMesh
    Socket Sampling Mode=::NewEnumerator5
   ) {module=SkeletalMeshLocation001}
   &CameraOffset(Camera Offset Amount=&UniformRangedFloat(Minimum=10, Maximum=10))
   &AddVelocity(Velocity=&RandomRangeVector(Minimum=<-2 -2 1>, Maximum=<2 2 3>))
  ]
  particleUpdateScript [
   &ParticleState()
   &Color(Color=&LerpLinearColors(
     StartColor=<rgba 1 1 1 1>
     EndColor=<rgba 0.7 0.7 0.7 1>
     LerpFactor=&RandomRangeFloat(Minimum=0, Maximum=1)
    ))
   &ScaleColor(
    Color Value To Scale=$Particles.Color
    Scale Alpha=&FloatFromCurve(
     FloatCurve=DataInterface() {propertyValues {Curve {keys [
         <at 0 0 RCIM_Cubic>
         {arriveTangent=0.2067, interpMode=RCIM_Cubic, leaveTangent=0.2067, tangentMode=RCTM_User, time=0.5, value=1}
         {arriveTangent=-1.3652, interpMode=RCIM_Cubic, leaveTangent=-1.3652, tangentMode=RCTM_User, time=0.8883, value=0.7994}
         <at 1 0>
        ]}}}
     Scale Curve=0.6
    )
    Scale RGB=&MakeVectorFromLinearColorRGB(Color=&MultiplyLinearColor(
      Linear Color=&LerpLinearColors(
       StartColor=$User.Color_01
       EndColor=$User.Color_02
       LerpFactor=&FloatFromCurve(FloatCurve=DataInterface() {propertyValues {Curve {keys [<at 0 0>,<at 0.9 0.15>]}}})
      )
      Scale Factor=&MakeVector4_FromLinearColor(LinearColor=&ColorFromCurve(ColorCurve=DataInterface() {propertyValues {
          AlphaCurve {keys [<at 0 1 RCIM_Cubic>,<at 0.9963 1>]}
          BlueCurve {keys [<at 0 1 RCIM_Cubic>,<at 1 1 RCIM_Cubic>]}
          GreenCurve {keys [<at 0 1 RCIM_Cubic>,<at 1 1 RCIM_Cubic>]}
          RedCurve {keys [<at 0 1 RCIM_Cubic>,<at 1 1 RCIM_Cubic>]}
         }}))
     ))
    ScaleA=true
    ScaleRGB=true
   )
   &ScaleColor(
    Color Value To Scale=$Particles.Color
    Scale RGB=&VectorFromFloat(Value=&FloatFromCurve(
      FloatCurve=DataInterface() {propertyValues {Curve {keys [<at 0 6 RCIM_Cubic>,<at 0.5 8 RCIM_Cubic>,<at 1 6 RCIM_Cubic>]}}}
      CurveIndex=Linked() {linkedVariable {name=NPC.FortniteNPC.TimeOfDay_0-1, type=NiagaraFloat}}
      Scale Curve=0.4
     ))
    ScaleA=false
    ScaleRGB=true
   ) {module=ScaleColor002}
   &ScaleColor(
    Color Value To Scale=$Particles.Color
    Scale Alpha=&FloatFromCurve(FloatCurve=DataInterface() {propertyValues {Curve {keys [<at 0 1.1>,<at 1 0.5>]}}}, CurveIndex=$System.FN_Integrator.NormalizedValue)
    Scale RGB=&VectorFromFloat(Value=&FloatFromCurve(
      FloatCurve=DataInterface() {propertyValues {Curve {keys [<at 0 0.85 RCIM_Cubic>, {arriveTangent=1.15, interpMode=RCIM_Cubic, leaveTangent=1.15, time=0.6869, value=1.144}, <at 1 2 RCIM_Cubic>]}}}
      CurveIndex=$System.FN_Integrator.NormalizedValue
     ))
    ScaleA=true
    ScaleRGB=true
   ) {module=ScaleColor001}
   &AccelerationForce(Acceleration=&Lerp_Vector(
     A=&RandomRangeVector(Minimum=<0 0 15>, Maximum=<0 0 30>)
     B=&RandomRangeVector(Minimum=<0 0 75>, Maximum=<0 0 125>)
     Alpha=&FloatFromCurve(
      FloatCurve=DataInterface() {propertyValues {Curve {keys [
          <at 0 0.2 RCIM_Cubic>
          {arriveTangent=0.6449, interpMode=RCIM_Cubic, leaveTangent=0.6449, time=0.4472, value=0.3712}
          {arriveTangent=1.1376, interpMode=RCIM_Cubic, leaveTangent=1.1376, time=0.7753, value=0.6999}
          <at 1 1 RCIM_Cubic>
         ]}}}
      CurveIndex=$System.FN_Integrator.NormalizedValue
     )
    ))
   &Drag(Drag=2, Use Linear Drag=true, Use Rotational Drag=false)
   &SolveForcesAndVelocity(Clamp Velocity=false, Limit Acceleration=false)
   &ScaleSpriteSize(Scale Factor=&Vector2DFromFloat(Value=&FloatFromCurve(
      FloatCurve=DataInterface() {propertyValues {Curve {keys [<at 0 1 RCIM_Cubic>, {arriveTangent=0.35, interpMode=RCIM_Cubic, leaveTangent=0.35, time=0.734, value=1.1055}, <at 1 1.35 RCIM_Cubic>]}}}
      CurveIndex=$System.FN_Integrator.NormalizedValue
     ))) {module=ScaleSpriteSize001}
   &DynamicMaterialParameters(
    Index 0 Param 2=0.8
    Index 0 Param 3=&RandomRangeFloat(Minimum=0, Maximum=1)
    Index 0 Param 4=&RandomRangeFloat(Minimum=0, Maximum=1)
    Module.Param0WriteEnabled=true
    Module.Param1WriteEnabled=true
    Module.Param2WriteEnabled=true
    Module.Param3WriteEnabled=true
   ) {module=DynamicMaterialParameters001}
  ]
  rendererClasses [&NiagaraSpriteRendererProperties]
  renderers [{rendererClass=&NiagaraSpriteRendererProperties}]
 }
 Fire_Light {
  emitterUpdateScript [
   &EmitterState(Life Cycle Mode=::NewEnumerator1, UseLoopDelay=false)
   &SpawnBurst_Instantaneous(Spawn Count=1, Spawn Probability=HlslExpression() {hlslExpression="Emitter.LoopCount < 1"}, Use Spawn Probability=true)
  ]
  particleSpawnScript [&InitializeParticle(
    Color=<rgba 0.6445 0.1981 0.6514 1>
    Module.Write Color=true
    Module.Write Lifetime=true
    Module.Write Mass=true
    Module.Write Material Random=true
    Module.Write Position=false
    Module.Write Scale=false
    Module.Write SpriteRotation=false
    Module.Write SpriteSize=true
    Sprite Size=&Multiply_Vector2(A=<50 40>, B=&Vector2DFromFloat(Value=0.8))
   )]
  particleUpdateScript [
   &ParticleState(
    Kill Particles When Lifetime Has Elapsed=false
    Let Infinitely Lived Particles Die When Emitter Deactivates=true
    Loop Particles Lifetime=true
   )
   &SkeletalMeshLocation(
    Apply Sampled Orientation to Mesh Particles=true
    Inherit Velocity=false
    MaskSpawnGroup=false
    Mesh Sampling Type=::NewEnumerator5
    Sampled Orientation Coordinate Space=::Local
    Sampled Position Coordinate Space=::Local
    Sampled Position Offset=<0 -7.5 7.5>
    Sampled Position Offset Coordinate Space:ENiagaraMeshTransforms=::NewEnumerator2
    Sampled Velocity Coordinate Space:ENiagaraCoordinateSpace=::Local
    Skeletal Mesh=$User.SkeletalMesh
    Socket Sampling Mode=::NewEnumerator5
   ) {module=SkeletalMeshLocation002}
   &Color(Color=$User.Color_01)
   &ScaleColor(
    Color Value To Scale=$Particles.Color
    Scale Alpha=&FloatFromCurve(
     FloatCurve=DataInterface() {propertyValues {Curve {keys [
         <at 0 0 RCIM_Cubic>
         {arriveTangent=0.8515, interpMode=RCIM_Cubic, leaveTangent=0.8515, time=0.1275, value=1}
         {arriveTangent=-1.1461, interpMode=RCIM_Cubic, leaveTangent=-1.1461, time=0.8566, value=0.7294}
         <at 1 0>
        ]}}}
     Scale Curve=2
    )
    Scale RGB=&VectorFromFloat(Value=&FloatFromCurve(
      FloatCurve=DataInterface() {propertyValues {Curve {keys [<at 0 1 RCIM_Cubic>, {arriveTangent=0.5, interpMode=RCIM_Cubic, leaveTangent=0.5, time=0.8592, value=1.35}, <at 1 1.5 RCIM_Cubic>]}}}
      CurveIndex=$System.FN_Integrator.NormalizedValue
     ))
    ScaleA=false
    ScaleRGB=true
   ) {module=ScaleColor001}
   /SetVariables(
    Particles.MeshColor:LinearColor=&MultiplyLinearColor(Linear Color=$Particles.Color, Scale Factor=<1 1 1 1>)
    Particles.SpriteColor=&MultiplyLinearColor(Linear Color=$Particles.Color, Scale Factor=<1 1 1 0.04>)
   ) {isSetParameters=true, module=SetVariables_#2}
   &DynamicMaterialParameters(
    Index 0 Param 2=1
    Index 0 Param 3=12
    Module.Param0WriteEnabled=true
    Module.Param1WriteEnabled=true
    Module.Param2WriteEnabled=true
    Module.Param3WriteEnabled=true
   )
   &FN_Light_Attributes(
    Color=&LerpLinearColors(StartColor=$User.Color_01, EndColor=$User.Color_02, LerpFactor=0.8)
    Create Color Attribute=true
    Create Exponent Attribute=true
    Create Radius Attribute=true
    Enable Particle Light Enable=true
    Exponent=2
    Light Enabled=true
    Radius=&FloatFromCurve(
     FloatCurve=DataInterface() {propertyValues {Curve {keys [<at 0 1 RCIM_Cubic>, {arriveTangent=0.4, interpMode=RCIM_Cubic, leaveTangent=0.4, time=0.6732, value=1.1856}, <at 1 1.4 RCIM_Cubic>]}}}
     CurveIndex=$System.FN_Integrator.NormalizedValue
     Scale Curve=25
    )
    Scale Color=&Multiply_Float(
     A=&FloatFromCurve(
      FloatCurve=DataInterface() {propertyValues {Curve {keys [
          <at 0 1 RCIM_Cubic>
          {arriveTangent=-0.6469, interpMode=RCIM_Cubic, leaveTangent=-0.6469, time=0.0926, value=1.134}
          {arriveTangent=0.5291, interpMode=RCIM_Cubic, leaveTangent=0.5291, time=0.2082, value=0.8653}
          {arriveTangent=0.0266, interpMode=RCIM_Cubic, leaveTangent=0.0266, time=0.3435, value=1.2667}
          {arriveTangent=-0.1747, interpMode=RCIM_Cubic, leaveTangent=-0.1747, time=0.5407, value=0.8741}
          {arriveTangent=2.3405, interpMode=RCIM_Cubic, leaveTangent=2.3405, time=0.6139, value=1.2195}
          {arriveTangent=-1.0967, interpMode=RCIM_Cubic, leaveTangent=-1.0967, time=0.722, value=1.2984}
          {arriveTangent=-1.0733, interpMode=RCIM_Cubic, leaveTangent=-1.0733, time=0.84, value=0.9716}
          <at 1 1 RCIM_Cubic>
         ]}}}
      Scale Curve=6
     )
     B=&FloatFromCurve(
      FloatCurve=DataInterface() {propertyValues {Curve {keys [<at 0 2 RCIM_Cubic>, {arriveTangent=4, interpMode=RCIM_Cubic, leaveTangent=4, time=0.6669, value=3.2773}, <at 1 6 RCIM_Cubic>]}}}
      CurveIndex=$System.FN_Integrator.NormalizedValue
     )
    )
   )
   &CameraOffset(Camera Offset Amount=-5)
  ]
  rendererClasses [&NiagaraLightRendererProperties]
  renderers [{rendererClass=&NiagaraLightRendererProperties, rendererIndex=0}]
 }
 Fire_Loop {
  emitterUpdateScript [
   &EmitterState()
   &SpawnRate(
    Spawn Probability=&FloatFromCurve(FloatCurve=DataInterface() {propertyValues {Curve {keys [<at 0 1>,<at 1 0.65>]}}}, CurveIndex=$System.FN_Integrator.NormalizedValue)
    SpawnRate=20
    Use Spawn Probability=true
   )
  ]
  particleSpawnScript [
   &InitializeParticle(
    Color=&RandomRangeLinearColor(Minimum=<rgba 0 0 0 1>, Maximum=<rgba 1 1 1 1>)
    Lifetime=&UniformRangedFloat(Minimum=0.9, Maximum=1.2)
    Module.Write Color=false
    Module.Write Lifetime=true
    Module.Write Mass=false
    Module.Write Material Random=true
    Module.Write Position=false
    Module.Write Scale=false
    Module.Write SpriteRotation=true
    Module.Write SpriteSize=true
    Position=&SimulationPosition()
    Sprite Rotation=&UniformRangedFloat(Minimum=-180, Maximum=180)
    Sprite Size=&Vector2DFromFloat(Value=&RandomRangeFloat(Minimum=15, Maximum=20))
   )
   &FN_SetSpawnGroup(Number of Spawn Groups=2)
   &SkeletalMeshLocation(
    Apply Sampled Orientation to Mesh Particles=true
    Bone / Socket Index=$Particles.SpawnGroup
    Inherit Velocity=true
    "Inherit Velocity (Scale)"=0.1
    MaskSpawnGroup=false
    Mesh Sampling Type=::NewEnumerator5
    Sampled Orientation Coordinate Space=::Local
    Sampled Position Coordinate Space=::Local
    Sampled Position Offset=&RandomRangeVector(Minimum=<0 -6 0>, Maximum=<0 6 0>)
    Sampled Position Offset Coordinate Space:ENiagaraMeshTransforms=::NewEnumerator3
    Sampled Velocity Coordinate Space:ENiagaraCoordinateSpace=::Local
    Skeletal Mesh=$User.SkeletalMesh
    Socket Sampling Mode=::NewEnumerator5
    VelocitySpeedLimit=false
   ) {module=SkeletalMeshLocation002}
   &CameraOffset(Camera Offset Amount=&UniformRangedFloat(Minimum=2.5, Maximum=5))
   &AddVelocity(Velocity=&RandomRangeVector(Minimum=<-2 -2 3>, Maximum=<2 2 5>))
  ]
  particleUpdateScript [
   &ParticleState()
   &Color(Color=&LerpLinearColors(
     StartColor=<rgba 1 1 1 1>
     EndColor=<rgba 0.7 0.7 0.7 1>
     LerpFactor=&RandomRangeFloat(Minimum=0, Maximum=1)
    ))
   &ScaleColor(
    Color Value To Scale=$Particles.Color
    Scale Alpha=&FloatFromCurve(FloatCurve=DataInterface() {propertyValues {Curve {keys [
         <at 0 0 RCIM_Cubic>
         {arriveTangent=0.2067, interpMode=RCIM_Cubic, leaveTangent=0.2067, tangentMode=RCTM_User, time=0.4, value=1}
         {arriveTangent=-1.3652, interpMode=RCIM_Cubic, leaveTangent=-1.3652, tangentMode=RCTM_User, time=0.8883, value=0.7994}
         <at 1 0>
        ]}}})
    Scale RGB=&MakeVectorFromLinearColorRGB(Color=&LerpLinearColors(
      StartColor=$User.Color_01
      EndColor=$User.Color_02
      LerpFactor=&FloatFromCurve(FloatCurve=DataInterface() {propertyValues {Curve {keys [<at 0 0>,<at 0.9 0.5>]}}})
     ))
    ScaleA=true
    ScaleRGB=true
   )
   &ScaleColor(Scale RGB=&VectorFromFloat(Value=1.5), ScaleA=false, ScaleRGB=true) {module=ScaleColor002}
   &ScaleColor(
    Color Value To Scale=$Particles.Color
    Scale Alpha=&FloatFromCurve(FloatCurve=DataInterface() {propertyValues {Curve {keys [<at 0 1.1>,<at 1 0.5>]}}}, CurveIndex=$System.FN_Integrator.NormalizedValue)
    Scale RGB=&VectorFromFloat(Value=&FloatFromCurve(
      FloatCurve=DataInterface() {propertyValues {Curve {keys [<at 0 0.85 RCIM_Cubic>, {arriveTangent=1.15, interpMode=RCIM_Cubic, leaveTangent=1.15, time=0.6869, value=1.144}, <at 1 2 RCIM_Cubic>]}}}
      CurveIndex=$System.FN_Integrator.NormalizedValue
     ))
    ScaleA=true
    ScaleRGB=true
   ) {module=ScaleColor001}
   &AccelerationForce(Acceleration=&Lerp_Vector(
     A=&RandomRangeVector(Minimum=<0 0 15>, Maximum=<0 0 25>)
     B=&RandomRangeVector(Minimum=<0 0 30>, Maximum=<0 0 40>)
     Alpha=&FloatFromCurve(
      FloatCurve=DataInterface() {propertyValues {Curve {keys [
          <at 0 0.2 RCIM_Cubic>
          {arriveTangent=0.6449, interpMode=RCIM_Cubic, leaveTangent=0.6449, time=0.4472, value=0.3712}
          {arriveTangent=1.1376, interpMode=RCIM_Cubic, leaveTangent=1.1376, time=0.7753, value=0.6999}
          <at 1 1 RCIM_Cubic>
         ]}}}
      CurveIndex=$System.FN_Integrator.NormalizedValue
     )
    ))
   &Drag(Drag=2, Use Linear Drag=true, Use Rotational Drag=false)
   &SolveForcesAndVelocity(Clamp Velocity=false, Limit Acceleration=false)
   /Niagara/Modules/Update/SubUV/SubUVAnimation(Number Of Frames=64)
   &ScaleSpriteSize(Scale Factor=&Vector2DFromFloat(Value=&FloatFromCurve(FloatCurve=DataInterface() {propertyValues {Curve {keys [
          {arriveTangent=-0.0625, interpMode=RCIM_Cubic, leaveTangent=-0.0625, tangentMode=RCTM_User, time=0, value=0}
          <at 0.4 1>
          {arriveTangent=0.3, interpMode=RCIM_Cubic, leaveTangent=0.3, time=0.6722, value=1}
          <at 1 0.3 RCIM_Cubic>
         ]}}})))
   &ScaleSpriteSize(Scale Factor=&Vector2DFromFloat(Value=&FloatFromCurve(
      FloatCurve=DataInterface() {propertyValues {Curve {keys [<at 0 1 RCIM_Cubic>, {arriveTangent=0.35, interpMode=RCIM_Cubic, leaveTangent=0.35, time=0.734, value=1.1055}, <at 1 1.35 RCIM_Cubic>]}}}
      CurveIndex=$System.FN_Integrator.NormalizedValue
     ))) {module=ScaleSpriteSize001}
   &DynamicMaterialParameters(
    Index 0 Param 1=&FloatFromCurve(FloatCurve=DataInterface() {propertyValues {Curve {keys [{arriveTangent=2.7551, interpMode=RCIM_Cubic, leaveTangent=2.7551, tangentMode=RCTM_User, time=0, value=0.25}, <at 0.7 1>]}}})
    Index 0 Param 2=1
    Index 0 Param 3=1.35
    Index 0 Param 4=1.2
    Module.Param0WriteEnabled=true
    Module.Param1WriteEnabled=true
    Module.Param2WriteEnabled=true
    Module.Param3WriteEnabled=true
   ) {module=DynamicMaterialParameters001}
  ]
  rendererClasses [&NiagaraSpriteRendererProperties]
  renderers [{rendererClass=&NiagaraSpriteRendererProperties}]
 }
 Sparks {
  emitterUpdateScript [
   &EmitterState()
   &SpawnRate(
    Spawn Probability=&Lerp_Float(
     A=0.2
     B=1.8
     Alpha=&FloatFromCurve(
      FloatCurve=DataInterface() {propertyValues {Curve {keys [
          <at 0 0 RCIM_Cubic>
          {arriveTangent=0.6534, interpMode=RCIM_Cubic, leaveTangent=0.6534, time=0.4418, value=0.1487}
          {arriveTangent=1.525, interpMode=RCIM_Cubic, leaveTangent=1.525, time=0.7641, value=0.4993}
          <at 1 1 RCIM_Cubic>
         ]}}}
      CurveIndex=$System.FN_Integrator.NormalizedValue
     )
    )
    SpawnRate=&Multiply_Float(A=60, B=0.6)
    Use Spawn Probability=true
   )
  ]
  particleSpawnScript [
   &InitializeParticle(
    Color=<rgba 1 2 3 1>
    Lifetime=&Lerp_Float(
     A=&RandomRangeFloat(Minimum=0.6, Maximum=0.8)
     B=&RandomRangeFloat(Minimum=0.6, Maximum=0.7)
     Alpha=&FloatFromCurve(
      FloatCurve=DataInterface() {propertyValues {Curve {keys [
          <at 0 0 RCIM_Cubic>
          {arriveTangent=0.6534, interpMode=RCIM_Cubic, leaveTangent=0.6534, time=0.4418, value=0.1487}
          {arriveTangent=1.525, interpMode=RCIM_Cubic, leaveTangent=1.525, time=0.7641, value=0.4993}
          <at 1 1 RCIM_Cubic>
         ]}}}
      CurveIndex=$System.FN_Integrator.NormalizedValue
     )
    )
    Module.Write Color=false
    Module.Write Lifetime=true
    Module.Write Mass=true
    Module.Write Material Random=true
    Module.Write Position=false
    Module.Write Scale=false
    Module.Write SpriteRotation=false
    Module.Write SpriteSize=true
    Position=&SimulationPosition()
    Sprite Size=&RandomRangeVector2D(Minimum=<0.8 2>, Maximum=<1.1 2.5>)
   )
   &FN_SetSpawnGroup(Number of Spawn Groups=2)
   &SkeletalMeshLocation(
    Apply Sampled Orientation to Mesh Particles=true
    Bone / Socket Index=$Particles.SpawnGroup
    Inherit Velocity=false
    MaskSpawnGroup=false
    Mesh Sampling Type=::NewEnumerator5
    Sampled Orientation Coordinate Space=::Local
    Sampled Position Coordinate Space=::Local
    Sampled Position Offset=&RandomRangeVector(Minimum=<-2 -8 0>, Maximum=<2 8 0>)
    Sampled Position Offset Coordinate Space:ENiagaraMeshTransforms=::NewEnumerator3
    Sampled Velocity Coordinate Space:ENiagaraCoordinateSpace=::Local
    Skeletal Mesh=$User.SkeletalMesh
    Socket Sampling Mode=::NewEnumerator5
   ) {module=SkeletalMeshLocation002}
   &AddVelocity(Velocity=&RandomRangeVector(Minimum=<-35 -35 20>, Maximum=<35 35 30>))
   &CameraOffset(Camera Offset Amount=2.5)
  ]
  particleUpdateScript [
   &ParticleState()
   &Drag(Use Linear Drag=true, Use Rotational Drag=false)
   &AccelerationForce(Acceleration=<0 0 25>)
   &CurlNoiseForce(Noise Frequency=100, Noise Strength=&RandomRangeFloat(Minimum=300, Maximum=500), Pan Noise=false)
   &SolveForcesAndVelocity(Clamp Velocity=false, Limit Acceleration=false)
   &ScaleSpriteSizeBySpeed(Max Scale Factor=<0.5 10>, Min Scale Factor=<0.5 0.5>)
   &ScaleSpriteSize(Scale Factor=&Vector2DFromFloat(Value=&FloatFromCurve(FloatCurve=DataInterface() {propertyValues {Curve {keys [{arriveTangent=-21.3532, interpMode=RCIM_Cubic, leaveTangent=-6.002, tangentMode=RCTM_User, time=0.5, value=1}, <at 1 0 RCIM_Cubic>]}}})))
   &Color(Color=&LerpLinearColors(
     StartColor=$System.Color_01
     EndColor=$System.Color_02
     LerpFactor=&FloatFromCurve(FloatCurve=DataInterface() {propertyValues {Curve {keys [<at 0 0>,<at 1 0.5>]}}})
    ))
   &ScaleColor(
    Color Value To Scale=$Particles.Color
    Scale Alpha=&FloatFromCurve(FloatCurve=DataInterface() {propertyValues {Curve {keys [<at 0 0 RCIM_Cubic>,<at 0.1 1 RCIM_Cubic>,<at 1 0>]}}})
    Scale RGB=&VectorFromFloat(Value=3)
    ScaleA=true
    ScaleRGB=true
   )
  ]
  rendererClasses [&NiagaraSpriteRendererProperties]
  renderers [{rendererClass=&NiagaraSpriteRendererProperties}]
 }
}
summary {
 systemName=NS_ScholarBackpack
 emitters [
  {emitterName=Sparks, rendererClasses [&NiagaraSpriteRendererProperties]}
  {emitterName=Fire_Loop, rendererClasses [&NiagaraSpriteRendererProperties]}
  {emitterName=FireAirglow, rendererClasses [&NiagaraSpriteRendererProperties]}
  {emitterName=FireMesh, rendererClasses [&NiagaraMeshRendererProperties]}
  {emitterName=Fire_Decals, rendererClasses [&NiagaraSpriteRendererProperties]}
  {emitterName=Fire_Light, rendererClasses [&NiagaraLightRendererProperties]}
 ]
 userVariables [
  User.bisStatChange {type=NiagaraBool}
  User.Color_01 {type=LinearColor, description="", defaultValue {struct=LinearColor, value=<rgba 0.7991 0.0103 0.7991 1>}}
  User.Color_02 {type=LinearColor, description="", defaultValue {struct=LinearColor, value=<rgba 0.227 0.0953 0.6514 1>}}
  User.RenderVis {type=NiagaraFloat, defaultValue=0, description=""}
  User.RibbonIDOffset {type=NiagaraInt32}
  User.SkeletalMesh {
   type=NiagaraDataInterfaceSkeletalMesh
   defaultValue {
    struct=NiagaraExt_VariableValue_DataInterface
    value {dataInterface=&self:NiagaraDataInterfaceSkeletalMesh_0}
   }
  }
  User.SKM_02 {
   type=NiagaraDataInterfaceSkeletalMesh
   defaultValue {
    struct=NiagaraExt_VariableValue_DataInterface
    value {dataInterface=&self:NiagaraDataInterfaceSkeletalMesh_1}
   }
  }
 ]
}
