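# Physics asset for the F_MED_Scholar backpack mesh, defining simulation bodies for its front cover and a hanging ribbon.
# Five box bodies cover the front cover bone and four ribbon bones, with thin, flat boxes along the ribbon.
# Constraints chain each ribbon segment to the previous one, rooted at the front cover.
# All joints share moderate swing limits and a tight twist limit, so the ribbon bends but barely twists.
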PhysicsAsset /Game/Accessories/FORT_Backpacks/Backpack_F_MED_Scholar/Meshes/F_MED_Scholar_Pack_Physics
bodies {
 JNT_Front_Cover [{shapeType=Box, extentX=3.1178, extentY=3.1178, extentZ=3.1178}]
 JNT_Ribbon_01 [{shapeType=Box, extentX=1.01, extentY=5.0497, extentZ=1.8506, center=<0 0 -1.0183>}]
 JNT_Ribbon_02 [{shapeType=Box, extentX=1.01, extentY=5.0936, extentZ=2.1299, center=<0 0 -1.066>}]
 JNT_Ribbon_03 [{shapeType=Box, extentX=1.01, extentY=5.0733, extentZ=1.45, center=<0 0 -0.6533>}]
 JNT_Ribbon_04 [{shapeType=Box, extentX=0.5, extentY=5.1615, extentZ=2.226, center=<0 0 -1.0218>}]
}
constraints [
 {bone1Name=JNT_Ribbon_02, bone2Name=JNT_Ribbon_01, swing1LimitDegrees=35, swing2LimitDegrees=35, twistLimitDegrees=5}
 {bone1Name=JNT_Ribbon_03, bone2Name=JNT_Ribbon_02, swing1LimitDegrees=35, swing2LimitDegrees=35, twistLimitDegrees=5}
 {bone1Name=JNT_Ribbon_04, bone2Name=JNT_Ribbon_03, swing1LimitDegrees=35, swing2LimitDegrees=35, twistLimitDegrees=5}
 {bone1Name=JNT_Ribbon_01, bone2Name=JNT_Front_Cover, swing1LimitDegrees=35, swing2LimitDegrees=35, twistLimitDegrees=5}
]
