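# Anim Blueprint for the Scholar backpack, a book that opens and flips pages as the player's AthenaKills stat rises.
# The Pack state machine handles open/close and up to four count-based flips; after three kills, a random toggle can enable looping left/right page turns.
# Curve-driven float additives and page turns feed RigidBody physics, then root scaling.
# Landing unlocks the book, skydiving resets it, and the front end simulates kills on timers.
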AnimBlueprint /BRCosmetics/Accessories/FORT_Backpacks/Backpack_F_MED_Scholar/Meshes/F_MED_Scholar_Pack_AnimBP
target_skeleton=/Game/Accessories/FORT_Backpacks/Backpack_F_MED_Scholar/Meshes/F_MED_Scholar_Pack_Skeleton
graphs [
 AnimGraph {nodes [
   Apply Additive(Additive="Sequence Player (additive) 'Scale_Pose_PROP'"() {sequence=/Game/Accessories/FORT_Backpacks/Backpack_F_MED_Scholar/Animation/Scale_Pose_PROP}) {alphaBoolBlend {blendOption=Linear}}
   Input Pose()
   AnimGraphNode_StateMachine_0 = Pack() {is=title AnimGraphNode_StateMachine}
   n1 = Apply Additive!(
    Base=Apply Additive(
     Base=@AnimGraphNode_StateMachine_0
     Additive="Sequence Player (additive) 'Idle_Open_Float_PROP'"() {sequence=/Game/Accessories/FORT_Backpacks/Backpack_F_MED_Scholar/Animation/Idle_Open_Float_PROP}
     AlphaCurveName=Floating
    ) {alphaInputType=Curve, alphaBoolBlend {blendOption=Linear}}
    Additive="Sequence Player (additive) 'Idle_Open_Float_Cover_PROP'"() {sequence=/Game/Accessories/FORT_Backpacks/Backpack_F_MED_Scholar/Animation/Idle_Open_Float_Cover_PROP}
    AlphaCurveName:Name=FullyOpen
   ) -> (Pose) {alphaInputType=Curve, alphaBoolBlend {blendOption=Linear}}
   AnimGraphNode_StateMachine_1 = PageTurning() {is=title AnimGraphNode_StateMachine}
   n2 = Blend Poses by bool(
    BlendPose_0=@AnimGraphNode_StateMachine_1
    BlendPose_1="Sequence Player (additive) 'Zero_Pose_PROP'"() {sequence=/Game/Accessories/FORT_Backpacks/Backpack_F_MED_Scholar/Animation/Zero_Pose_PROP}
    bActiveValue=Get UseMultiPageStyle() -> (UseMultiPageStyle:bool)
   ) {blendType=Linear}
   n3 = Get PageTurnAlpha() -> (PageTurnAlpha:double)
   n4 = RigidBody(ComponentPose=Local To Component(LocalPose=Apply Additive!(Base=@n1, Additive=@n2, Alpha:float=@n3) -> (Pose) {alphaBoolBlend {blendOption=Linear}})) {
    bClampLinearTranslationLimitToRefPose=true
    bOverrideWorldGravity=true
    overridePhysicsAsset=/Game/Accessories/FORT_Backpacks/Backpack_F_MED_Scholar/Meshes/F_MED_Scholar_Pack_Physics
    simulationSpace=BaseBoneSpace
    alphaBoolBlend {blendOption=Linear}
    baseBoneRef {boneName=root}
    componentAppliedLinearAccClamp=<3000 3000 3000>
    componentLinearAccScale=<0.25 0.25 0.25>
    componentLinearVelScale=<0.25 0.25 0.25>
    overrideWorldGravity {z=-980}
   }
   n5 = "Transform (Modify) Bone - Bone: root"(ComponentPose=@n4, Translation="0,0,0", Rotation="0,0,0", Scale="0.9,0.9,0.9") {
    scaleMode=BMM_Replace
    scaleSpace=BCS_ParentBoneSpace
    alphaBoolBlend {blendOption=Linear}
    boneToModify {boneName=root}
    scale=<0.9 0.9 0.9>
   }
   Output Pose(Result=Component To Local(ComponentPose=@n5)) {layerGroup=None}
  ]}
 Pack {
  entry=Closed
  path=AnimGraph.AnimGraphNode_StateMachine_0.Pack
  ids=id
  nodes [
   AnimStateNode_0 = Closed() {is=title AnimStateNode}
   AnimStateNode_1 = Open() {is=title AnimStateNode}
   AnimStateNode_2 = Open_Into() {is=title AnimStateNode}
   AnimStateNode_3 = Open_Out() {is=title AnimStateNode}
   AnimStateNode_4 = PageFlip_page1() {is=title AnimStateNode}
   AnimStateNode_5 = PageFlip_page2() {is=title AnimStateNode}
   AnimStateNode_6 = PageFlip_page3() {is=title AnimStateNode}
   AnimStateNode_7 = PageFlip_page4() {is=title AnimStateNode}
   AnimStateTransitionNode_0 = AnimStateTransitionNode()
   AnimStateTransitionNode_1 = AnimStateTransitionNode() {bAutomaticRuleBasedOnSequencePlayerInState=true, crossfadeDuration=0.05}
   AnimStateTransitionNode_12 = AnimStateTransitionNode() {crossfadeDuration=0}
   AnimStateTransitionNode_13 = AnimStateTransitionNode()
   AnimStateTransitionNode_14 = AnimStateTransitionNode()
   AnimStateTransitionNode_15 = AnimStateTransitionNode()
   AnimStateTransitionNode_16 = AnimStateTransitionNode()
   AnimStateTransitionNode_2 = AnimStateTransitionNode()
   AnimStateTransitionNode_3 = AnimStateTransitionNode() {bAutomaticRuleBasedOnSequencePlayerInState=true, crossfadeDuration=0.05}
   AnimStateTransitionNode_5 = AnimStateTransitionNode() {crossfadeDuration=0}
   AnimStateTransitionNode_6 = AnimStateTransitionNode() {crossfadeDuration=0}
   AnimStateTransitionNode_7 = AnimStateTransitionNode() {crossfadeDuration=0}
  ]
  links [
   {
    dst=Open_Into
    rule=AnimGraph.AnimGraphNode_StateMachine_0.Pack.AnimStateTransitionNode_0.Transition
    src=Closed
    via=AnimStateTransitionNode_0
   }
   {dst=Open_Out, rule=AnimGraph.AnimGraphNode_StateMachine_0.Pack.AnimStateTransitionNode_2.Transition, src=Open, via=AnimStateTransitionNode_2}
   {
    dst=PageFlip_page1
    rule=AnimGraph.AnimGraphNode_StateMachine_0.Pack.AnimStateTransitionNode_12.Transition
    src=Open
    via=AnimStateTransitionNode_12
   }
   {
    dst=Open
    rule=AnimGraph.AnimGraphNode_StateMachine_0.Pack.AnimStateTransitionNode_1.Transition
    src=Open_Into
    via=AnimStateTransitionNode_1
   }
   {
    dst=Closed
    rule=AnimGraph.AnimGraphNode_StateMachine_0.Pack.AnimStateTransitionNode_3.Transition
    src=Open_Out
    via=AnimStateTransitionNode_3
   }
   {
    dst=Open_Out
    rule=AnimGraph.AnimGraphNode_StateMachine_0.Pack.AnimStateTransitionNode_13.Transition
    src=PageFlip_page1
    via=AnimStateTransitionNode_13
   }
   {
    dst=PageFlip_page2
    rule=AnimGraph.AnimGraphNode_StateMachine_0.Pack.AnimStateTransitionNode_5.Transition
    src=PageFlip_page1
    via=AnimStateTransitionNode_5
   }
   {
    dst=Open_Out
    rule=AnimGraph.AnimGraphNode_StateMachine_0.Pack.AnimStateTransitionNode_14.Transition
    src=PageFlip_page2
    via=AnimStateTransitionNode_14
   }
   {
    dst=PageFlip_page3
    rule=AnimGraph.AnimGraphNode_StateMachine_0.Pack.AnimStateTransitionNode_6.Transition
    src=PageFlip_page2
    via=AnimStateTransitionNode_6
   }
   {
    dst=Open_Out
    rule=AnimGraph.AnimGraphNode_StateMachine_0.Pack.AnimStateTransitionNode_15.Transition
    src=PageFlip_page3
    via=AnimStateTransitionNode_15
   }
   {
    dst=PageFlip_page4
    rule=AnimGraph.AnimGraphNode_StateMachine_0.Pack.AnimStateTransitionNode_7.Transition
    src=PageFlip_page3
    via=AnimStateTransitionNode_7
   }
   {
    dst=Open_Out
    rule=AnimGraph.AnimGraphNode_StateMachine_0.Pack.AnimStateTransitionNode_16.Transition
    src=PageFlip_page4
    via=AnimStateTransitionNode_16
   }
  ]
 }
 Closed {
  path=AnimGraph.AnimGraphNode_StateMachine_0.Pack.AnimStateNode_0.Closed
  nodes [Output Animation Pose(Result=Sequence Player 'Idle_Closed_PROP'() {sequence=/Game/Accessories/FORT_Backpacks/Backpack_F_MED_Scholar/Animation/Idle_Closed_PROP})]
 }
 Open {
  path=AnimGraph.AnimGraphNode_StateMachine_0.Pack.AnimStateNode_1.Open
  nodes [Output Animation Pose(Result=Sequence Player 'Idle_Open_PROP'() {sequence=/Game/Accessories/FORT_Backpacks/Backpack_F_MED_Scholar/Animation/Idle_Open_PROP})]
 }
 Open_Into {
  path=AnimGraph.AnimGraphNode_StateMachine_0.Pack.AnimStateNode_2.Open_Into
  nodes [Output Animation Pose(Result=Sequence Player 'Idle_Open_Into_PROP'() {bLoopAnimation=false, sequence=/Game/Accessories/FORT_Backpacks/Backpack_F_MED_Scholar/Animation/Idle_Open_Into_PROP})]
 }
 Open_Out {
  path=AnimGraph.AnimGraphNode_StateMachine_0.Pack.AnimStateNode_3.Open_Out
  nodes [Output Animation Pose(Result=Sequence Player 'Idle_Open_Out_PROP'() {bLoopAnimation=false, sequence=/Game/Accessories/FORT_Backpacks/Backpack_F_MED_Scholar/Animation/Idle_Open_Out_PROP})]
 }
 PageFlip_page1 {
  path=AnimGraph.AnimGraphNode_StateMachine_0.Pack.AnimStateNode_4.PageFlip_page1
  nodes [Output Animation Pose(Result=Sequence Player 'PageFlip_R2L_page1_PROP'!(PlayRate:float=1) -> (Pose) {bLoopAnimation=false, sequence=/Game/Accessories/FORT_Backpacks/Backpack_F_MED_Scholar/Animation/PageFlip_R2L_page1_PROP})]
 }
 PageFlip_page2 {
  path=AnimGraph.AnimGraphNode_StateMachine_0.Pack.AnimStateNode_5.PageFlip_page2
  nodes [Output Animation Pose(Result=Sequence Player 'PageFlip_R2L_page2_PROP'!(PlayRate:float=1) -> (Pose) {bLoopAnimation=false, sequence=/Game/Accessories/FORT_Backpacks/Backpack_F_MED_Scholar/Animation/PageFlip_R2L_page2_PROP})]
 }
 PageFlip_page3 {
  path=AnimGraph.AnimGraphNode_StateMachine_0.Pack.AnimStateNode_6.PageFlip_page3
  nodes [Output Animation Pose(Result=Sequence Player 'PageFlip_R2L_page3_PROP'!(PlayRate:float=1) -> (Pose) {bLoopAnimation=false, sequence=/Game/Accessories/FORT_Backpacks/Backpack_F_MED_Scholar/Animation/PageFlip_R2L_page3_PROP})]
 }
 PageFlip_page4 {
  path=AnimGraph.AnimGraphNode_StateMachine_0.Pack.AnimStateNode_7.PageFlip_page4
  nodes [Output Animation Pose(Result=Sequence Player 'PageFlip_R2L_page4_PROP'!(PlayRate:float=1) -> (Pose) {bLoopAnimation=false, sequence=/Game/Accessories/FORT_Backpacks/Backpack_F_MED_Scholar/Animation/PageFlip_R2L_page4_PROP})]
 }
 Transition {
  path=AnimGraph.AnimGraphNode_StateMachine_0.Pack.AnimStateTransitionNode_0.Transition
  nodes [Result(bCanEnterTransition=Get IsOpen() -> (IsOpen:bool))]
 }
 Transition {path=AnimGraph.AnimGraphNode_StateMachine_0.Pack.AnimStateTransitionNode_1.Transition, nodes [Result()]}
 Transition {
  path=AnimGraph.AnimGraphNode_StateMachine_0.Pack.AnimStateTransitionNode_12.Transition
  nodes [
   Result(bCanEnterTransition=Get UsePageCountStyle() -> (UsePageCountStyle:bool))
   AND Boolean!(A=Get Kill() -> (Kill:bool), B=Get FullyOpen() -> (FullyOpen:bool), C:bool=NOT Boolean(A=Get MultiPageStyle() -> (MultiPageStyle:bool))) -> (ReturnValue)
  ]
 }
 Transition {
  path=AnimGraph.AnimGraphNode_StateMachine_0.Pack.AnimStateTransitionNode_13.Transition
  nodes [
   n1 = Get KillCount() -> (KillCount:int=0)
   n2 = "Time Remaining (PageFlip_R2L_page1_PROP)"!() -> (ReturnValue:float) {is=title K2Node_AnimGetter}
   Result(bCanEnterTransition=AND Boolean!(
     A="Equal (Integer)"(A=@n1, B=1) {is=title K2Node_CallFunction}
     B="Equal (Float)"!(A:double=@n2, B:double=0) -> (ReturnValue:bool) {is=title K2Node_CallFunction}
    ) -> (ReturnValue))
  ]
 }
 Transition {
  path=AnimGraph.AnimGraphNode_StateMachine_0.Pack.AnimStateTransitionNode_14.Transition
  nodes [
   n1 = Get KillCount() -> (KillCount=0)
   n2 = "Time Remaining (PageFlip_R2L_page2_PROP)"!() -> (ReturnValue:float) {is=title K2Node_AnimGetter}
   Result(bCanEnterTransition=AND Boolean(A="Equal (Integer)"(A=@n1, B=2), B="Equal (Float)"(A=@n2, B=0)))
  ]
 }
 Transition {
  path=AnimGraph.AnimGraphNode_StateMachine_0.Pack.AnimStateTransitionNode_15.Transition
  nodes [
   n1 = Get KillCount() -> (KillCount=0)
   n2 = "Time Remaining (PageFlip_R2L_page3_PROP)"!() -> (ReturnValue:float) {is=title K2Node_AnimGetter}
   Result(bCanEnterTransition=AND Boolean(A="Equal (Integer)"(A=@n1, B=3), B="Equal (Float)"(A=@n2, B=0)))
  ]
 }
 Transition {
  path=AnimGraph.AnimGraphNode_StateMachine_0.Pack.AnimStateTransitionNode_16.Transition
  nodes [
   n1 = "Time Remaining (PageFlip_R2L_page4_PROP)"!() -> (ReturnValue:float) {is=title K2Node_AnimGetter}
   n2 = Get KillCount() -> (KillCount=0)
   Result(bCanEnterTransition=AND Boolean(A="integer >= integer"!(A:int=@n2, B:int=4) -> (ReturnValue:bool) {is=title K2Node_CallFunction}, B="Equal (Float)"(A=@n1, B=0)))
  ]
 }
 Transition {
  path=AnimGraph.AnimGraphNode_StateMachine_0.Pack.AnimStateTransitionNode_2.Transition
  nodes [Result(bCanEnterTransition=NOT Boolean(A=Get IsOpen()))]
 }
 Transition {path=AnimGraph.AnimGraphNode_StateMachine_0.Pack.AnimStateTransitionNode_3.Transition, nodes [Result()]}
 Transition {
  path=AnimGraph.AnimGraphNode_StateMachine_0.Pack.AnimStateTransitionNode_5.Transition
  nodes [
   n1 = "Time Remaining (PageFlip_R2L_page1_PROP)"()
   n2 = Get KillCount() -> (KillCount=0)
   Result(bCanEnterTransition=AND Boolean(A="integer >= integer"(A=@n2, B=2), B="Equal (Float)"(A=@n1, B=0)))
  ]
 }
 Transition {
  path=AnimGraph.AnimGraphNode_StateMachine_0.Pack.AnimStateTransitionNode_6.Transition
  nodes [
   n1 = "Time Remaining (PageFlip_R2L_page2_PROP)"()
   n2 = Get KillCount() -> (KillCount=0)
   Result(bCanEnterTransition=AND Boolean(A="integer >= integer"(A=@n2, B=3), B="Equal (Float)"(A=@n1, B=0)))
  ]
 }
 Transition {
  path=AnimGraph.AnimGraphNode_StateMachine_0.Pack.AnimStateTransitionNode_7.Transition
  nodes [
   n1 = Get KillCount() -> (KillCount=0)
   n2 = "Time Remaining (PageFlip_R2L_page3_PROP)"()
   Result(bCanEnterTransition=AND Boolean(A="integer >= integer"(A=@n1, B=4), B="Equal (Float)"(A=@n2, B=0)))
  ]
 }
 PageTurning {
  entry=PageTurning
  path=AnimGraph.AnimGraphNode_StateMachine_1.PageTurning
  ids=id
  nodes [AnimStateNode_0 = PageTurning() {is=title AnimStateNode}]
 }
 PageTurning {
  path=AnimGraph.AnimGraphNode_StateMachine_1.PageTurning.AnimStateNode_0.PageTurning
  nodes [Output Animation Pose(Result="Sequence Player (additive) 'Idle_Page_Turning_RightToLeft_PROP'"!(
     Sequence:Anim Sequence Base Object Reference=Get PageTurnAnim() -> (PageTurnAnim:Anim Sequence Base Object Reference)
     PlayRate:float=Get PageFlipSpeed_MutliPages() -> (PageFlipSpeed_MutliPages:double)
    ) -> (Pose) {playRate=2})]
 }
 BoolRules {nodes [
   Inputs()
   > n1 = Sequence()
   > Set UseMultiPageStyle(UseMultiPageStyle:bool=AND Boolean(A=Get MultiPageStyle(), B=Get FullyOpen())) -> (Output_Get:bool)
   n2 = Get MultiPageStyle()
   > Set UsePageCountStyle(UsePageCountStyle:bool=AND Boolean!(A=Get Kill(), B=Get FullyOpen(), C:bool=NOT Boolean(A=@n2)) -> (ReturnValue)) -> (Output_Get:bool)
   Outputs!(then:Exec=@n1.then_1)
  ]}
 EventGraph {nodes [
   Event BlueprintUpdateAnimation()
   > Set FullyOpen(FullyOpen:bool="Equal (Float)"(A=GetCurveValue(CurveName=FullyOpen), B=1)) -> (Output_Get:bool)
   > Bool Rules!(execute:Exec) -> (then:Exec) {is=title K2Node_MacroInstance}
   > n1 = Sequence!(execute) -> (then_0, then_1, then_2:Exec)
   > Kill Counter!(execute:Exec) -> (then:Exec) {is=title K2Node_MacroInstance}
   n2 = Branch(execute=@n1.then_1, Condition=Get bIsSkyDiving())
   > Set IsOpen(IsOpen:bool=false) -> (Output_Get:bool)
   > Set UnlockBook(UnlockBook:bool=false) -> (Output_Get:bool)
   > Set PageTurnAlpha(PageTurnAlpha:double=0) -> (Output_Get:double)
   > Set Timer(Timer:double=0) -> (Output_Get:double)
   > Set Kill(Kill:bool=false) -> (Output_Get:bool)
   Branch(execute=@n2.else, Condition=OR Boolean(A=Get bIsOnGround() -> (bIsOnGround:bool), B=Get bIsInFrontEnd()))
   > Set UnlockBook(UnlockBook=true)
   n3 = Branch(execute=@n1.then_2, Condition=Get bIsInFrontEnd())
   > Random Method!(execute:Exec) -> (then:Exec) {is=title K2Node_MacroInstance}
   > Kill Method!(execute:Exec) -> (then:Exec) {is=title K2Node_MacroInstance}
   Kill Method(execute=@n3.else)
   AnimNotify_PageTurnOff() {is=title K2Node_Event}
   > Set Kill(Kill=false)
   > Set IsOpen(IsOpen=false)
   > n4 = Branch(Condition="integer >= integer"(A=Get KillCount() -> (KillCount=0), B=3))
   > n5 = Set MultiPageStyle(MultiPageStyle:bool=Select(Option 0:bool=false, Option 1:bool=true, Index:int=RandomIntegerInRange!(Min:int=0, Max:int=1) -> (ReturnValue:int=0)) -> (ReturnValue:bool)) -> (Output_Get:bool)
   Set PageTurnAlpha(execute=[@n5.then, @n4.else], PageTurnAlpha=0)
   > n6 = Branch(Condition="Equal (Integer)"(A=Get LeftRight() -> (LeftRight:int=0), B=0))
   > Set LeftRight(LeftRight:int=1) -> (Output_Get:int=0)
   Set LeftRight(execute=@n6.else, LeftRight=0) -> (Output_Get=0)
   AnimNotify_ReactivityCue() {is=title K2Node_Event}
   > Call Reactivity Cue!(execute:Exec, self:F MED Scholar Pack Anim BP Object Reference) -> (then:Exec) {is=title K2Node_CallDelegate}
   Event BlueprintInitializeAnimation()
   n7 = TryGetFortPawn()
   > Is Valid(InputObject=@n7)
   > n8 = Cast To FortPlayerPawn(Object=@n7)
   > Set Pawn(Pawn:Fort Player Pawn Object Reference=@n8) -> (Output_Get:Fort Player Pawn Object Reference)
   > n9 = WatchCosmeticStat!(
    execute:Exec
    Pawn:Fort Player Pawn Object Reference=Get Pawn() -> (Pawn:Fort Player Pawn Object Reference)
    StatTag:Fort Stat Manager Tag Structure=(TagName="StatManager.AthenaKills")
    DefaultValue:int=0
    bBroadcastInitialValue:bool=false
    InitialValueBroadcastDelay:float=-1
    ChangeValueBroadcastDelay:float=-1
    bConsumeWaitingNotifications:bool=true
   ) -> (then:Exec, OnStatChanged:Exec, StatName:Name, StatValue:int) {is=title K2Node_AsyncAction}
   n10 = Set KillCount(
    execute=@n9.OnStatChanged
    KillCount:int=Select(Option 0:int=@n9.StatValue, Option 1:int=Get KillCount() -> (KillCount=0), Index:bool=Get bIsInFrontEnd()) -> (ReturnValue:int)
   ) -> (Output_Get:int=0)
   > Branch(Condition="integer > integer"(A=@n10, B=Get KillCount_Last() -> (KillCount_Last:int=0)) {is=title K2Node_CallFunction})
   > Set Kill(Kill=true)
   > Set KillCount_Last(KillCount_Last:int=Get KillCount() -> (KillCount=0)) -> (Output_Get:int=0)
   n11 = "To Float (Integer)"(InInt=Get KillCount() -> (KillCount=0))
   n12 = "Equal (Float)"(A=@n11, B=0)
   > Set PageFlipSpeed(PageFlipSpeed:double=MapRangeClamped(Value=SelectFloat(A=1, B=@n11, bPickA=@n12), InRangeA=1, InRangeB=6, OutRangeA=1.75, OutRangeB=2.5)) -> (Output_Get:double)
   TryGetPawnOwner()
   PrintString!(execute, InString="To String (Integer)"(InInt=Get KillCount() -> (KillCount=0)), bPrintToScreen, bPrintToLog, TextColor, Duration, Key) -> (then)
   Get bIsMoving2D()
  ]}
 KillCounter {nodes [
   n1 = WatchCosmeticStat(
    Pawn=Get Pawn()
    StatTag=(TagName="StatManager.AthenaKills")
    DefaultValue=0
    bBroadcastInitialValue=false
    InitialValueBroadcastDelay=-1
    ChangeValueBroadcastDelay=-1
    bConsumeWaitingNotifications=true
   )
   n2 = Set KillCount(execute=@n1.OnStatChanged, KillCount=Select(Option 0=@n1.StatValue, Option 1=Get KillCount() -> (KillCount=0), Index=Get bIsInFrontEnd())) -> (Output_Get=0)
   > Branch(Condition="integer > integer"(A=@n2, B=Get KillCount_Last() -> (KillCount_Last=0)))
   > Set Kill(Kill=true)
   > Set KillCount_Last(KillCount_Last=Get KillCount() -> (KillCount=0)) -> (Output_Get=0)
   n3 = "To Float (Integer)"(InInt=Get KillCount() -> (KillCount=0))
   n4 = "Equal (Float)"(A=@n3, B=0)
   > Set PageFlipSpeed(PageFlipSpeed=MapRangeClamped(Value=SelectFloat(A=1, B=@n3, bPickA=@n4), InRangeA=1, InRangeB=6, OutRangeA=1.75, OutRangeB=2.5))
   Inputs()
   > n5 = Sequence()
   n6 = Get KillCount() -> (KillCount=0)
   Set Kill_1(execute=@n5.then_1, Kill_1:bool="integer >= integer"(A=@n6, B=1)) -> (Output_Get:bool)
   > Set Kill_2(Kill_2:bool="integer >= integer"(A=@n6, B=2)) -> (Output_Get:bool)
   > Set Kill_3(Kill_3:bool="integer >= integer"(A=@n6, B=3)) -> (Output_Get:bool)
   > Set Kill_4(Kill_4:bool="integer >= integer"(A=@n6, B=4)) -> (Output_Get:bool)
   Outputs(then=@n5.then_2)
  ]}
 KillMethod {nodes [
   n1 = Branch(Condition="Equal (Integer)"(A=Get LeftRight() -> (LeftRight=0), B=0))
   > Set LeftRight(LeftRight=1) -> (Output_Get=0)
   Set LeftRight(execute=@n1.else, LeftRight=0) -> (Output_Get=0)
   Inputs()
   > n2 = Sequence!(execute) -> (then_0, then_1)
   > Set PageFlipSpeed_MutliPages(PageFlipSpeed_MutliPages:double=Get PageFlipSpeed_New() -> (PageFlipSpeed_New:double)) -> (Output_Get:double)
   n3 = Get bIsSkyDiving()
   > Branch(Condition=AND Boolean!(A=Get UnlockBook() -> (UnlockBook:bool), B=NOT Boolean(A=@n3)) -> (ReturnValue))
   > n4 = Branch(Condition=Get Kill())
   > Set IsOpen(IsOpen=true)
   > Branch(Condition=Get MultiPageStyle())
   n5 = Get LeftRight() -> (LeftRight=0)
   > Set PageTurnAnim(PageTurnAnim:Anim Sequence Base Object Reference=Select(
     Option 0:Anim Sequence Base Object Reference=/Game/Accessories/FORT_Backpacks/Backpack_F_MED_Scholar/Animation/Idle_Page_Turning_RightToLeft_PROP
     Option 1:Anim Sequence Base Object Reference=/Game/Accessories/FORT_Backpacks/Backpack_F_MED_Scholar/Animation/Idle_Page_Turning_LeftToRight_PROP
     Index:int=@n5
    ) -> (ReturnValue:Anim Sequence Base Object Reference)) -> (Output_Get:Anim Sequence Base Object Reference)
   > Set PageTurnAlpha(PageTurnAlpha=1)
   > Delay(Duration=float * float(A=Select(Option 0:double=4.2, Option 1:double=4.2, Index=@n5) -> (ReturnValue:double), B=Get PageFlipSpeed_MutliPages()))
   > Set PageFlipSpeed_New(PageFlipSpeed_New:double=RandomFloatInRange(Min=1.15, Max=1.5)) -> (Output_Get:double)
   Set IsOpen(execute=@n4.else, IsOpen=false)
   Outputs(then=@n2.then_1)
  ]}
 RandomMethod {nodes [
   n1 = Branch(Condition=Get Kill())
   > Set IsOpen(IsOpen=true)
   > Set Timer(Timer=10)
   > Branch(Condition=Get MultiPageStyle())
   > Set PageTurnAnim(PageTurnAnim=Select(
     Option 0:Anim Sequence Base Object Reference=&Idle_Page_Turning_RightToLeft_PROP
     Option 1:Anim Sequence Base Object Reference=&Idle_Page_Turning_LeftToRight_PROP
     Index=Get LeftRight() -> (LeftRight=0)
    ) -> (ReturnValue:Anim Sequence Base Object Reference))
   > Delay(Duration=4)
   > Set PageTurnAlpha(PageTurnAlpha=1)
   > n2 = Branch(Condition="Equal (Integer)"(A=Get LeftRight() -> (LeftRight=0), B=0))
   > Set LeftRight(LeftRight=1) -> (Output_Get=0)
   Set LeftRight(execute=@n2.else, LeftRight=0) -> (Output_Get=0)
   Set IsOpen(execute=@n1.else, IsOpen=false)
   > Set Timer(Timer=7)
   n3 = Branch(Condition=Get IsOpen())
   > Set IsOpen(IsOpen=false)
   > Set Timer(Timer=7)
   Set IsOpen(execute=@n3.else, IsOpen=true)
   > Set Timer(Timer=10)
   Inputs()
   > n4 = Sequence()
   > Set PageFlipSpeed(PageFlipSpeed=1)
   n5 = Get bIsSkyDiving()
   > Branch(Condition=AND Boolean(A=Get UnlockBook(), B=NOT Boolean(A=@n5)))
   > Delay(Duration=Get Timer() -> (Timer:double))
   > Increment Int(Value=Get KillCount() -> (KillCount=0))
   > Set Timer(Timer=RandomFloatInRange(Min=10, Max=10))
   Outputs(then=@n4.then_1)
  ]}
 ReactivityCue {nodes [ReactivityCue() {is=title K2Node_FunctionEntry}]}
]
variables [
 FullyOpen {type=boolean}
 IsOpen {type=boolean}
 Kill {type=boolean}
 Kill_1 {type=boolean}
 Kill_2 {type=boolean}
 Kill_3 {type=boolean}
 Kill_4 {type=boolean}
 KillCount {type=integer}
 KillCount_Last {type=integer}
 LeftRight {type=integer}
 MultiPageStyle {type=boolean}
 PageFlipSpeed {type=number, default=1}
 PageFlipSpeed_MutliPages {type=number, default=1.25}
 PageFlipSpeed_New {type=number, default=1.25}
 PageTurnAlpha {type=number}
 PageTurnAnim {type=AnimSequenceBase, default=&Idle_Page_Turning_RightToLeft_PROP}
 Pawn {type=FortPlayerPawn}
 ReactivityCue {default=()}
 Timer {type=number, default=2}
 UnlockBook {type=boolean}
 UseMultiPageStyle {type=boolean}
 UsePageCountStyle {type=boolean}
]
